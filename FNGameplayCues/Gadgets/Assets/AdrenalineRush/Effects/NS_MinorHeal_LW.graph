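# NS_MinorHeal_LW is a Niagara system for a minor heal effect among the Adrenaline Rush gadget gameplay cue assets.
# It contains two emitters, SparkleStationary and one named "+s", whose name suggests plus symbols alongside sparkles.
# The system uses the SystemState module, and the file lists no renderers, data interfaces, dynamic inputs or user parameters for it.
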NiagaraSystem /FNGameplayCues/Gadgets/Assets/AdrenalineRush/Effects/NS_MinorHeal_LW
dependencies {usedDataInterfaces [], usedDynamicInputs [], usedModules [&SystemState]}
emitters {+s {rendererClasses [], renderers []}, SparkleStationary {rendererClasses [], renderers []}}
summary {
 systemName=NS_MinorHeal_LW
 emitters [{emitterName=SparkleStationary, rendererClasses []}, {emitterName=+s, rendererClasses []}]
 userVariables []
}
